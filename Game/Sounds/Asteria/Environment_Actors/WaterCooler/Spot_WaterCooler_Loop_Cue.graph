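# Looping indoor ambient sound cue for a water cooler, combining a steady hum with intermittent bubbles.
# A mixer blends a looping, slightly pitched-down hum with a louder bubble branch.
# The bubble branch repeatedly waits a random delay, then plays one of four equally weighted bubble waves through a modulator for variation.
# The cue plays at low volume with hum attenuation, a reverb submix send and a prevent-new concurrency override.
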SoundCue /Game/Sounds/Asteria/Environment_Actors/WaterCooler/Spot_WaterCooler_Loop_Cue
root=Mixer_0
nodes [
 {id=Delay_0, children [Modulator_0], properties {delayMax=8, delayMin=0.6}}
 {id=Looping_0, children [Delay_0]}
 {id=Mixer_0, children [Looping_0,Modulator_6], properties {inputVolume [0.7,1.25]}}
 {id=Modulator_0, children [Random_1]}
 {id=Modulator_6, children [WavePlayer_5], properties {pitchMax=0.95, pitchMin=0.85}}
 {id=Random_1, children [WavePlayer_6,WavePlayer_7,WavePlayer_8,WavePlayer_9], properties {weights [1,1,1,1]}}
 {
  id=WavePlayer_5
  properties {bLooping=true, soundWaveAssetPtr=&/Prop_Water_Cooler_Hum_Loop}
 }
 {id=WavePlayer_6, properties {soundWaveAssetPtr=&/Prop_Water_Cooler_Bubble_01}}
 {id=WavePlayer_7, properties {soundWaveAssetPtr=&/Prop_Water_Cooler_Bubble_02}}
 {id=WavePlayer_8, properties {soundWaveAssetPtr=&/Prop_Water_Cooler_Bubble_03}}
 {id=WavePlayer_9, properties {soundWaveAssetPtr=&/Prop_Water_Cooler_Bubble_04}}
]
properties {
 attenuationSettings=&/Spot_WaterCooler_Hum_Att
 bOverrideConcurrency=true
 maxDistance=750
 soundClassObject=/Game/Sounds/FortSoundClassesAndModes/Fort_Ambient_OneShots_Indoors
 volumeMultiplier=0.2
 attenuationOverrides {falloffDistance=2000}
 concurrencyOverrides {resolutionRule=PreventNew}
 concurrencySet [&Weapon_Drop_StopOldest]
 soundSubmixSends [{minSendLevel=0.5, soundSubmix=&Fortnite_Reverb_Default}]
}
